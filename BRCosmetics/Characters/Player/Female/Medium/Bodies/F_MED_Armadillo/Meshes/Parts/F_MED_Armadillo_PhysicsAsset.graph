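# Physics asset for the F_MED_Armadillo female medium player cosmetic, covering only the head and dynamic hair bones for secondary hair motion.
# A compound sphere-and-capsule head body anchors single-capsule hair bodies through swing- and twist-limited constraints.
# Each side has a three-link bang chain that loosens past the root, a loose strand, and a tightly limited bun.
# Bun strands branch directly from their bun, two left and three right.
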PhysicsAsset /BRCosmetics/Characters/Player/Female/Medium/Bodies/F_MED_Armadillo/Meshes/Parts/F_MED_Armadillo_PhysicsAsset
bodies {
 hair_l_bang_01_dyn [{shapeType=Capsule, length=2.5, radius=0.75, shapeName=hair_l_bang_01_dyn_capsule, center=<2 0 0>, rotation=<rot 90 0 0>}]
 hair_l_bang_02_dyn [{shapeType=Capsule, length=1.25, radius=0.75, shapeName=hair_l_bang_02_dyn_capsule, center=<1.5 0 0>, rotation=<rot 90 0 0>}]
 hair_l_bang_03_dyn [{shapeType=Capsule, length=1.5, radius=0.75, shapeName=hair_l_bang_03_dyn_capsule, center=<1.75 0 0>, rotation=<rot 90 0 0>}]
 hair_l_bun_dyn [{shapeType=Capsule, length=0.5, radius=3, shapeName=hair_l_bun_dyn_capsule, center=<2 0 0>, rotation=<rot 90 0 0>}]
 hair_l_bun_strand_01_dyn [{shapeType=Capsule, length=1, radius=0.5, shapeName=hair_l_bun_strand_01_dyn_capsule, center=<1 0 0>, rotation=<rot 90 0 0>}]
 hair_l_bun_strand_02_dyn [{shapeType=Capsule, length=2, radius=0.5, shapeName=hair_l_bun_strand_02_dyn_capsule, center=<1 0 0>, rotation=<rot 90 0 0>}]
 hair_l_strand_01_dyn [{shapeType=Capsule, length=2, radius=0.75, shapeName=hair_l_strand_01_dyn_capsule, center=<2 0 0>, rotation=<rot 90 0 0>}]
 hair_r_bang_01_dyn [{shapeType=Capsule, length=2.5, radius=0.75, shapeName=hair_r_bang_01_dyn_capsule, center=<2 0 0>, rotation=<rot 90 0 0>}]
 hair_r_bang_02_dyn [{shapeType=Capsule, length=1.25, radius=0.75, shapeName=hair_r_bang_02_dyn_capsule, center=<1.5 0 0>, rotation=<rot 90 0 0>}]
 hair_r_bang_03_dyn [{shapeType=Capsule, length=1.5, radius=0.75, shapeName=hair_r_bang_03_dyn_capsule, center=<1.75 0 0>, rotation=<rot 90 0 0>}]
 hair_r_bun_dyn [{shapeType=Capsule, length=0.5, radius=3, shapeName=hair_r_bun_dyn_capsule, center=<2 0 0>, rotation=<rot 90 0 0>}]
 hair_r_bun_strand_01_dyn [{shapeType=Capsule, length=1, radius=0.5, shapeName=hair_r_bun_strand_01_dyn_capsule, center=<1 0 0>, rotation=<rot 90 0 0>}]
 hair_r_bun_strand_02_dyn [{shapeType=Capsule, length=1, radius=0.5, shapeName=hair_r_bun_strand_02_dyn_capsule, center=<1 0 0>, rotation=<rot 90 0 0>}]
 hair_r_bun_strand_03_dyn [{shapeType=Capsule, length=1, radius=0.75, shapeName=hair_r_bun_strand_03_dyn_capsule, center=<1 0 0>, rotation=<rot 90 0 0>}]
 hair_r_strand_01_dyn [{shapeType=Capsule, length=2, radius=0.75, shapeName=hair_r_strand_01_dyn_capsule, center=<2 0 0>, rotation=<rot 90 0 0>}]
 head [
  {shapeType=Sphere, radius=7.4, shapeName=head_sphere, center=<6.5 -1 0.0896>}
  {shapeType=Capsule, length=8, radius=6, center=<2 4 0>, rotation=<rot 0 75 90>}
  {shapeType=Capsule, length=1, radius=2, shapeName=head_capsule, center=<2.5 -0.275 -5.5>, rotation=<rot 0 80 90>}
  {shapeType=Capsule, length=1, radius=2, shapeName=head_capsule, center=<2.5 -0.275 5.5>, rotation=<rot 0 80 90>}
 ]
}
constraints [
 {bone1Name=hair_l_bang_01_dyn, bone2Name=head, swing1LimitDegrees=20, swing2LimitDegrees=20, twistLimitDegrees=15}
 {bone1Name=hair_l_bang_02_dyn, bone2Name=hair_l_bang_01_dyn, swing1LimitDegrees=30, swing2LimitDegrees=30, twistLimitDegrees=15}
 {bone1Name=hair_l_bang_03_dyn, bone2Name=hair_l_bang_02_dyn, swing1LimitDegrees=30, swing2LimitDegrees=30, twistLimitDegrees=15}
 {bone1Name=hair_r_bang_01_dyn, bone2Name=head, swing1LimitDegrees=20, swing2LimitDegrees=20, twistLimitDegrees=15}
 {bone1Name=hair_r_bang_02_dyn, bone2Name=hair_r_bang_01_dyn, swing1LimitDegrees=30, swing2LimitDegrees=30, twistLimitDegrees=15}
 {bone1Name=hair_r_bang_03_dyn, bone2Name=hair_r_bang_02_dyn, swing1LimitDegrees=30, swing2LimitDegrees=30, twistLimitDegrees=15}
 {bone1Name=hair_l_bun_dyn, bone2Name=head, swing1LimitDegrees=10, swing2LimitDegrees=10, twistLimitDegrees=3}
 {bone1Name=hair_l_bun_strand_01_dyn, bone2Name=hair_l_bun_dyn, swing1LimitDegrees=15, swing2LimitDegrees=15, twistLimitDegrees=10}
 {bone1Name=hair_l_bun_strand_02_dyn, bone2Name=hair_l_bun_dyn, swing1LimitDegrees=15, swing2LimitDegrees=15, twistLimitDegrees=5}
 {bone1Name=hair_l_strand_01_dyn, bone2Name=head, swing1LimitDegrees=15, swing2LimitDegrees=15, twistLimitDegrees=15}
 {bone1Name=hair_r_bun_dyn, bone2Name=head, swing1LimitDegrees=10, swing2LimitDegrees=10, twistLimitDegrees=3}
 {bone1Name=hair_r_bun_strand_01_dyn, bone2Name=hair_r_bun_dyn, swing1LimitDegrees=10, swing2LimitDegrees=10, twistLimitDegrees=5}
 {bone1Name=hair_r_bun_strand_02_dyn, bone2Name=hair_r_bun_dyn, swing1LimitDegrees=20, swing2LimitDegrees=20, twistLimitDegrees=15}
 {bone1Name=hair_r_bun_strand_03_dyn, bone2Name=hair_r_bun_dyn, swing1LimitDegrees=20, swing2LimitDegrees=20, twistLimitDegrees=15}
 {bone1Name=hair_r_strand_01_dyn, bone2Name=head, swing1LimitDegrees=15, swing2LimitDegrees=15, twistLimitDegrees=15}
]
